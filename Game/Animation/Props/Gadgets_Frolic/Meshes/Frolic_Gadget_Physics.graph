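# Physics asset for the Frolic gadget prop mesh, giving its skeleton a simple collision shape.
# It holds a single body on the JNT_Card bone, a small rotated capsule with length equal to its radius.
# There are no constraints, so no joints link bodies together.
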PhysicsAsset /Game/Animation/Props/Gadgets_Frolic/Meshes/Frolic_Gadget_Physics
bodies {JNT_Card [{shapeType=Capsule, length=0.505, radius=0.505, rotation=<rot 0 -90 0>}]}
constraints []
